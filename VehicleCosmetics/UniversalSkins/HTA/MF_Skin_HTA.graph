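# Material function for the HTA universal vehicle skin, outputting material attributes for clear-coated car paint with emissive.
# A skin mask, rebuilt from the texture's inverted blue channel, blends team, accent and paint colors, optional diffuse, and finish values.
# Time-driven sine/cosine offsets produce masked, animated UV distortion, while camera-angle math shifts between two tints for an iridescent-style color.
# A SharpMasks switch, marked porting-only, reproduces Rocket League mask sharpening.
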MaterialFunction /VehicleCosmetics/UniversalSkins/HTA/MF_Skin_HTA
graph {nodes [
  n1 = NamedRerouteUsage()
  Composite!(SkinMaskTexture=@n1) {subgraphName=BodyFinish}
  n2 = NamedRerouteUsage()
  MaterialExpressionReroute_29 = Reroute(None=@n2)
  n3 = NamedRerouteUsage()
  MaterialExpressionReroute_8 = Reroute(None=@n3)
  n4 = NamedRerouteUsage()
  MaterialExpressionReroute_13 = Reroute(None=@n4)
  n5 = NamedRerouteUsage()
  MaterialExpressionReroute_20 = Reroute(None=@n5)
  n6 = MaterialFunctionCall(Color=@MaterialExpressionReroute_20.SkinMaskTexture) {materialFunction=&SplitComponents}
  n7 = NamedRerouteUsage()
  MaterialExpressionReroute_18 = Reroute(None=@n7)
  n8 = LinearInterpolate(
   A=LinearInterpolate(A=@MaterialExpressionReroute_8.TeamColor, B=@MaterialExpressionReroute_13.AccentColor, Alpha=@n6.G)
   B=@MaterialExpressionReroute_18.PaintColor
   Alpha=@n6.B
  )
  MaterialExpressionReroute_26 = Reroute(None=@n8)
  MaterialExpressionReroute_27 = Reroute(None=@MaterialExpressionReroute_26.BaseColor)
  n9 = NamedRerouteUsage()
  MaterialExpressionReroute_30 = Reroute(None=@n9)
  n10 = StaticSwitchParameter(
   True=LinearInterpolate(
    A=@MaterialExpressionReroute_29.BaseColorTexture
    B=@MaterialExpressionReroute_27.BaseColor
    Alpha=ComponentMask(None=@MaterialExpressionReroute_30.SkinMaskTexture) {r=true}
   ) {desc=Put diffuse where R skin mask is black}
   False=@MaterialExpressionReroute_27.BaseColor
  ) {defaultValue=true, parameterName=UseBaseColorTexture}
  MaterialExpressionReroute_28 = Reroute(None=@n10)
  n11 = NamedRerouteUsage()
  Composite!(BaseColorInput=@MaterialExpressionReroute_28.BaseColor, TextureInput=@n11) {desc=Legacy - Rocket League Feature, subgraphName=SharpMasksBaseColorMultiply}
  n12 = NamedRerouteUsage()
  n13 = NamedRerouteUsage()
  n14 = NamedRerouteUsage()
  Composite!(DetailBaseTexture=@n12, DetailAccentTexture=@n13, SkinMaskTexture=@n14) {subgraphName=Details}
  n15 = NamedRerouteUsage()
  n16 = NamedRerouteUsage()
  Composite!(Roughness=@n15, SkinMask=@n16) {subgraphName=Roughness}
  Composite!(TeamColor=@n3, AccentColor=@n4, PaintColor=@n7, SkinMaskTexture=@n5) {subgraphName=BaseColor}
  Composite!(BaseColor=@MaterialExpressionReroute_26.BaseColor, BaseColorTexture=@n2, SkinMaskTexture=@n9) {subgraphName=UseBaseColorTexture}
  Composite!() {subgraphName=Vectors}
  n17 = NamedRerouteUsage()
  n18 = NamedRerouteUsage()
  Composite!(SkinMaskTexture=@n17, BottomNormalBaseTexture=@n18, BottomNormalAccentTexture=@n18) {subgraphName=ClearCoat}
  n19 = Constant()
  n20 = TextureSampleParameter2D(UVs=NamedRerouteUsage()) {
   group=Texture
   parameterName=Mask_UVDistortion
   samplerSource=SSM_Wrap_WorldGroupSettings
   samplerType=SAMPLERTYPE_LinearColor
   texture=&/Skin_HTA_RGBA
  }
  n21 = MaterialFunctionCall(X=@n19, Y=OneMinus(None=@n20.B), Z=@n19, A=@n19) {materialFunction=&MakeFloat4}
  Composite!(RGB=@n21.Result) {desc=Legacy - Rocket League Feature, subgraphName=ProcessSharpMasks}
  Constant3Vector() {constant=<rgba 0 0 1.6 0>}
  Constant() {r=0.45}
  Constant() {r=0.575}
  Constant() {r=0.85}
  n22 = SetMaterialAttributes!(
   MaterialAttributes
   BaseColor=NamedRerouteUsage()
   Metallic=NamedRerouteUsage()
   Roughness=NamedRerouteUsage()
   CustomData0=NamedRerouteUsage()
   CustomData1=NamedRerouteUsage()
   Normal=NamedRerouteUsage()
   ClearCoatBottomNormal=NamedRerouteUsage()
   EmissiveColor=NamedRerouteUsage()
  ) {attributeSetTypes [
    #1
    #2
    #3
    #4
    #5
    #6
    #7
    #8
   ]}
  FunctionOutput(None=@n22)
  MaterialExpressionReroute_10 = Reroute(None=@n1)
  n23 = MaterialFunctionCall(Color=@MaterialExpressionReroute_10.SkinMaskTexture) {materialFunction=&SplitComponents}
  n24 = LinearInterpolate(
   A=ScalarParameter() {defaultValue=0.85, group=Finish, parameterName=Metallic_AccentColor}
   B=ScalarParameter() {defaultValue=0.7, group=Finish, parameterName=Metallic_TeamColor}
   Alpha=@n23.G
  )
  LinearInterpolate(A=@n24, B=ScalarParameter() {defaultValue=0.45, group=Finish, parameterName=Metallic_PaintColor}, Alpha=@n23.B)
  MaterialExpressionReroute_11 = Reroute(None=@n17)
  n25 = MaterialFunctionCall(Color=@MaterialExpressionReroute_11.SkinMaskTexture) {materialFunction=&SplitComponents}
  LinearInterpolate(
   A=LinearInterpolate(
    A=ScalarParameter() {defaultValue=1, group=Finish, parameterName=CCThickness_TeamColor}
    B=ScalarParameter() {defaultValue=1, group=Finish, parameterName=CCThickness_AccentColor}
    Alpha=@n25.G
   )
   B=ScalarParameter() {defaultValue=1, group=Finish, parameterName=CCThickness_PaintColor}
   Alpha=@n25.B
  )
  MaterialExpressionReroute_35 = Reroute(None=@n18)
  MaterialExpressionReroute_36 = Reroute(None=@n18)
  n26 = MaterialFunctionCall(Color=@MaterialExpressionReroute_11.SkinMaskTexture) {materialFunction=&SplitComponents}
  LinearInterpolate(
   A=LinearInterpolate(A=@MaterialExpressionReroute_35.BottomNormalBaseTexture, B=@MaterialExpressionReroute_36.BottomNormalAccentTexture, Alpha=@n26.G)
   B=@MaterialExpressionReroute_36.BottomNormalAccentTexture
   Alpha=@n26.B
  )
  n27 = MaterialFunctionCall(Color=@MaterialExpressionReroute_11.SkinMaskTexture) {materialFunction=&SplitComponents}
  LinearInterpolate(
   A=LinearInterpolate(
    A=ScalarParameter() {defaultValue=0.1, group=Finish, parameterName=CCRoughness_TeamColor}
    B=ScalarParameter() {defaultValue=0.4, group=Finish, parameterName=CCRoughness_AccentColor}
    Alpha=@n27.G
   )
   B=ScalarParameter() {defaultValue=0.4, group=Finish, parameterName=CCRoughness_PaintColor}
   Alpha=@n27.B
  )
  n28 = MaterialFunctionCall(Color=@MaterialExpressionReroute_10.SkinMaskTexture) {materialFunction=&SplitComponents}
  n29 = LinearInterpolate(
   A=ScalarParameter() {defaultValue=0.95, group=Finish, parameterName=Rough_AccentColor}
   B=ScalarParameter() {defaultValue=0.7, group=Finish, parameterName=Rough_TeamColor}
   Alpha=@n28.G
  )
  LinearInterpolate(A=@n29, B=ScalarParameter() {defaultValue=0.55, group=Finish, parameterName=Rough_PaintColor}, Alpha=@n28.B)
  n30 = NamedRerouteUsage()
  n31 = Multiply(A=Constant3Vector() {constant=<rgba 0.192 0.4545 0.831 0>}, B=@n30)
  n32 = Multiply(A=@n30, B=Constant3Vector() {constant=<rgba 0.4215 0.7182 0.7493 0>})
  n33 = NamedRerouteUsage()
  n34 = Transform(None=CameraVectorWS()) {transformSourceType=TRANSFORMSOURCE_World, transformType=TRANSFORM_Tangent}
  n35 = Constant3Vector() {constant=<rgba 0 0 1.25 0>}
  n36 = OneMinus(None=Saturate(None=DotProduct(A=@n34, B=@n35)))
  n37 = DotProduct(
   A=LinearInterpolate(A=Constant3Vector() {constant=<rgba 0 0 2.8 0>}, B=@n33, Alpha=LinearInterpolate(Alpha=@n36) {constA=0.3}) {constAlpha=0.3}
   B=@n34
  )
  n38 = Power(Base=@n37) {constExponent=1.4}
  n39 = Power(Base=@n33) {constExponent=50}
  n40 = Constant3Vector() {constant=<rgba 0 0 5 0>}
  n41 = OneMinus(None=Saturate(None=DotProduct(A=@n34, B=@n40)))
  n42 = Saturate(None=DotProduct(A=@n37, B=Multiply(A=@n39, B=@n41)))
  n43 = Add(A=LinearInterpolate(B=Clamp(None=@n38) {maxDefault=1.1}) {constA=0.95, constAlpha=3}, B=Saturate(None=Multiply(A=@n42) {constB=7}))
  n44 = NamedRerouteUsage()
  n45 = MaterialFunctionCall(Float2=NamedRerouteUsage()) {materialFunction=&BreakOutFloat2Components}
  n46 = LinearInterpolate(
   A=NamedRerouteUsage()
   B=LinearInterpolate(A=NamedRerouteUsage(), B=LinearInterpolate(A=@n31, B=@n32, Alpha=@n43), Alpha=ComponentMask(None=@n44) {g=true})
   Alpha=Max(A=@n45.R, B=@n45.R)
  ) {desc=use color}
  Multiply(B=@n46) {constA=0.025, desc=diffuse rim}
  NamedRerouteDeclaration(None=TextureCoordinate() {coordinateIndex=1}) {name=UV, nodeColor=<rgba 1 0.8646 0.8646 1>}
  MaterialExpressionReroute_42 = Reroute(None=@n21.Result)
  n47 = Power(Base=@MaterialExpressionReroute_42.RGB) {constExponent=0.4545}
  n48 = StaticSwitchParameter(True=Saturate(None=ConstantBiasScale(None=@n47) {bias=-0.5, scale=4}), False=@MaterialExpressionReroute_42.RGB) {
   desc="DO NOT USE FOR NEW ASSETS! This is intended for porting content from Rocket League only."
   group=Legacy_PORTINGONLY
   parameterName=SharpMasks
  }
  MaterialExpressionReroute_45 = Reroute(None=@n48)
  NamedRerouteDeclaration(None=@MaterialExpressionReroute_45.SharpMasksSwitchOutput) {name=SkinMaskModified, nodeColor=<rgba 0.6038 0.3325 0.013 1>}
  n49 = VectorParameter() {group=Color, parameterName=BaseColor, sortPriority=1, defaultValue=<rgba 0 0.2 1 1>}
  MaterialExpressionReroute_23 = Reroute(None=@n49.RGB)
  NamedRerouteDeclaration(None=@MaterialExpressionReroute_23.TeamColor) {name=TeamColor, nodeColor=<rgba 0.9219 0 1 1>}
  n50 = VectorParameter() {group=Color, parameterName=AccentColor, sortPriority=2, defaultValue=<rgba 1 1 1 1>}
  MaterialExpressionReroute_24 = Reroute(None=@n50.RGB)
  NamedRerouteDeclaration(None=@MaterialExpressionReroute_24.AccentColor) {name=AccentColor, nodeColor=<rgba 0.9219 0 1 1>}
  n51 = VectorParameter() {group=Color, parameterName=PaintColor, sortPriority=3, defaultValue=<rgba 0 1 0 1>}
  MaterialExpressionReroute_25 = Reroute(None=@n51.RGB)
  NamedRerouteDeclaration(None=@MaterialExpressionReroute_25.PaintColor) {name=PaintColor, nodeColor=<rgba 0.9219 0 1 1>}
  n52 = TextureSampleParameter2D(UVs=NamedRerouteUsage()) {group=Finish, parameterName=Normal, samplerType=SAMPLERTYPE_Normal, texture=&/Skin_HTA_N}
  n53 = MaterialFunctionCall(Normal=@n52.RGB, Flatness=ScalarParameter() {defaultValue=0.775, group=DetailNormal, parameterName=BottomNormalFlatness}) {materialFunction=&FlattenNormal}
  NamedRerouteDeclaration(None=@n53.Result) {name=BottomNormalBaseTexture, nodeColor=<rgba 0.6038 0.3325 0.013 1>}
  n54 = TextureSampleParameter2D(UVs=NamedRerouteUsage()) {group=Texture, parameterName=BaseColor, samplerSource=SSM_Wrap_WorldGroupSettings, texture=&/Skin_HTA_D}
  NamedRerouteDeclaration(None=@n54.RGB) {name=BaseColorTexture, nodeColor=<rgba 0.6038 0.3325 0.013 1>}
  n55 = NamedRerouteUsage()
  n56 = NamedRerouteUsage()
  NamedRerouteDeclaration(None=Multiply(A=Multiply(A=@n55, B=@n56), B=NamedRerouteUsage())) {name=BaseColor, nodeColor=<rgba 0 1 0.7656 1>}
  NamedRerouteDeclaration(None=NamedRerouteUsage()) {name=CCThicknessMasked, nodeColor=<rgba 0 1 0.7656 1>}
  NamedRerouteDeclaration(None=NamedRerouteUsage()) {name=CCRoughnessMasked, nodeColor=<rgba 0 1 0.7656 1>}
  MaterialExpressionReroute_6 = Reroute(None=@n24)
  NamedRerouteDeclaration(None=@MaterialExpressionReroute_6.Metallic) {name=Metallic, nodeColor=<rgba 0 0.1434 1 1>}
  NamedRerouteDeclaration(None=@n53.Result) {name=BottomNormalAccentTexture, nodeColor=<rgba 0.6038 0.3325 0.013 1>}
  n57 = Constant() {r=1}
  NamedRerouteDeclaration(None=@n57) {name=AOTexture, nodeColor=<rgba 0.6038 0.3325 0.013 1>}
  n58 = TextureSampleParameter2D() {group=Texture, parameterName=BodyNormalLOD, samplerType=SAMPLERTYPE_Normal, texture=/VehicleCosmetics/Textures/T_BlankNormal}
  NamedRerouteDeclaration(None=LinearInterpolate(A=Constant3Vector() {constant=<rgba 0 0 1 0>}, B=@n58.RGB, Alpha=ScalarParameter() {parameterName=LOD}) {constAlpha=0}) {name=BodyNormalLOD, nodeColor=<rgba 0.6038 0.3325 0.013 1>}
  n59 = MaterialFunctionCall(BaseNormal=NamedRerouteUsage(), AdditionalNormal=@MaterialExpressionReroute_35.BottomNormalBaseTexture) {materialFunction=&BlendAngleCorrectedNormals}
  MaterialExpressionReroute_37 = Reroute(None=@n59.Result)
  NamedRerouteDeclaration(None=@MaterialExpressionReroute_37.ClearCoatBottomNormal) {name=CCBottomNormal, nodeColor=<rgba 0 0.1434 1 1>}
  NamedRerouteDeclaration(None=Add(A=@n46, B=LinearInterpolate(A=@n46, Alpha=@n45.G) {constA=1, constB=0.025})) {name=PreBaseColor, nodeColor=<rgba 1 0.2109 0 1>}
  MaterialExpressionReroute_46 = Reroute(None=@MaterialExpressionReroute_42.RGB)
  NamedRerouteDeclaration(None=@MaterialExpressionReroute_46.TextureRawOutput) {name=SkinMaskModifiedRaw, nodeColor=<rgba 1 0 0.2812 1>}
  n60 = MaterialFunctionCall(Float4=NamedRerouteUsage()) {materialFunction=&BreakOutFloat4Components}
  n61 = Multiply(A=OneMinus(None=@n60.R), B=@n60.G)
  n62 = Multiply(A=@n60.R, B=OneMinus(None=@n60.G))
  n63 = Saturate(None=@n60.A)
  n64 = Multiply(A=@n63) {constB=0.5}
  n65 = Multiply(A=Time()) {constB=0.5}
  n66 = Multiply(A=@n63)
  n67 = AppendVector(A=Sine(None=Add(A=@n64, B=@n65)), B=Cosine(None=Add(A=@n66, B=@n65)))
  NamedRerouteDeclaration(None=Multiply(A=AppendVector(A=@n61, B=@n62), B=Multiply(A=@n67) {constB=0.007})) {name=Distortion, nodeColor=<rgba 0 0.3437 1 1>}
  NamedRerouteDeclaration(None=NamedRerouteDeclaration(None=NamedRerouteUsage()) {name=MetallicMasked, nodeColor=<rgba 0 1 0.7656 1>}) {name=MetallicOutput, nodeColor=<rgba 0 1 0.7647 1>}
  MaterialExpressionReroute_2 = Reroute(None=@n15)
  n68 = MaterialFunctionCall(Color=@MaterialExpressionReroute_2.Roughness) {materialFunction=&SplitComponents}
  MaterialExpressionReroute_3 = Reroute(None=@n16)
  n69 = MaterialFunctionCall(Color=@MaterialExpressionReroute_3.SkinMask) {materialFunction=&SplitComponents}
  n70 = LinearInterpolate(A=LinearInterpolate(A=@n68.R, B=@n68.R, Alpha=@n69.G), B=@n68.B, Alpha=@n69.B)
  MaterialExpressionReroute_5 = Reroute(None=@n70)
  NamedRerouteDeclaration(None=NamedRerouteDeclaration(None=@MaterialExpressionReroute_5.RoughnessMasked) {name=RoughnessMasked, nodeColor=<rgba 0 1 0.7656 1>}) {name=RoughnessOutput, nodeColor=<rgba 0 1 0.7647 1>}
  MaterialExpressionReroute_9 = Reroute(None=@n29)
  NamedRerouteDeclaration(None=@MaterialExpressionReroute_9.Roughness) {name=Roughness, nodeColor=<rgba 0 0.1434 1 1>}
  n71 = NamedRerouteUsage()
  n72 = NamedRerouteUsage()
  n73 = NamedRerouteUsage()
  NamedRerouteDeclaration(None=LinearInterpolate(A=Add(A=@n71, B=@n72), B=@n72, Alpha=ComponentMask(None=@n73) {g=true})) {name=UVs_Distorted, nodeColor=<rgba 0.6641 1 0 1>}
  n74 = NamedRerouteUsage()
  NamedRerouteDeclaration(None=LinearInterpolate(A=Multiply(A=@n74) {constB=0.65}, Alpha=@n45.G) {constB=0}) {name=EmissiveColor, nodeColor=<rgba 0.4531 1 0 1>}
  NamedRerouteDeclaration(None=@n20.RGBA) {name=SkinMask, nodeColor=<rgba 0.6039 0.3333 0.0118 1>}
  NamedRerouteDeclaration(None=@n57) {name=Curvature, nodeColor=<rgba 0.6039 0.3333 0.0118 1>}
  n75 = Constant()
  MaterialExpressionReroute_32 = Reroute(None=@n75)
  NamedRerouteDeclaration(None=@MaterialExpressionReroute_32.ClearCoat Thickness) {name=CCThickness, nodeColor=<rgba 0 0.1434 1 1>}
  MaterialExpressionReroute_33 = Reroute(None=@n75)
  NamedRerouteDeclaration(None=@MaterialExpressionReroute_33.ClearCoat Roughness) {name=CCRoughness, nodeColor=<rgba 0 0.1434 1 1>}
  NamedRerouteDeclaration(None=@n57) {name=DetailBaseTexture, nodeColor=<rgba 0.6038 0.3325 0.013 1>}
  NamedRerouteDeclaration(None=@n57) {name=DetailAccentTexture, nodeColor=<rgba 0.6038 0.3325 0.013 1>}
  MaterialExpressionReroute_31 = Reroute(None=@n12)
  MaterialExpressionReroute_34 = Reroute(None=@n13)
  MaterialExpressionReroute_16 = Reroute(None=@n14)
  n76 = LinearInterpolate(
   A=@MaterialExpressionReroute_31.DetailBaseTexture
   B=@MaterialExpressionReroute_34.DetailAccentTexture
   Alpha=Saturate(None=MaterialFunctionCall(Color=@MaterialExpressionReroute_16.SkinMaskTexture) {materialFunction=&SplitComponents}.G)
  ) {constA=1}
  MaterialExpressionReroute_17 = Reroute(None=@n76)
  NamedRerouteDeclaration(None=@MaterialExpressionReroute_17.DetailsMasked) {name=DetailsMasked, nodeColor=<rgba 0 1 0.7656 1>}
  NamedRerouteUsage()
  PinBase() {reroutePins [SkinMaskTexture {expression=&self:MaterialExpressionReroute_10}]}
  PinBase!(Metallic=@n24, Roughness=@n29) {reroutePins [
    Metallic {expression=&self:MaterialExpressionReroute_6}
    Roughness {expression=&self:MaterialExpressionReroute_9}
   ]}
  PinBase!() {reroutePins [
    BaseColor {expression=&self:MaterialExpressionReroute_27}
    BaseColorTexture {expression=&self:MaterialExpressionReroute_29}
    SkinMaskTexture {expression=&self:MaterialExpressionReroute_30}
   ]}
  PinBase!(BaseColor=@n10) {reroutePins [BaseColor {expression=&self:MaterialExpressionReroute_28}]}
  PinBase!() {reroutePins [
    SkinMaskTexture {expression=&self:MaterialExpressionReroute_11}
    BottomNormalBaseTexture {expression=&self:MaterialExpressionReroute_35}
    BottomNormalAccentTexture {expression=&self:MaterialExpressionReroute_36}
   ]}
  PinBase!(ClearCoat Thickness=@n75, ClearCoat Roughness=@n75, ClearCoatBottomNormal=@n59.Result) {reroutePins [
    ClearCoat Thickness {expression=&self:MaterialExpressionReroute_32}
    ClearCoat Roughness {expression=&self:MaterialExpressionReroute_33}
    ClearCoatBottomNormal {expression=&self:MaterialExpressionReroute_37}
   ]}
  PinBase!() {reroutePins [
    DetailBaseTexture {expression=&self:MaterialExpressionReroute_31}
    DetailAccentTexture {expression=&self:MaterialExpressionReroute_34}
    SkinMaskTexture {expression=&self:MaterialExpressionReroute_16}
   ]}
  PinBase!(DetailsMasked=@n76) {reroutePins [DetailsMasked {expression=&self:MaterialExpressionReroute_17}]}
  PinBase!() {reroutePins [RGB {expression=&self:MaterialExpressionReroute_42}]}
  PinBase!(SharpMasksSwitchOutput=@n48, TextureRawOutput=@MaterialExpressionReroute_42.RGB) {reroutePins [
    SharpMasksSwitchOutput {expression=&self:MaterialExpressionReroute_45}
    TextureRawOutput {expression=&self:MaterialExpressionReroute_46}
   ]}
  PinBase!() {reroutePins [
    BaseColorInput {expression=&self:MaterialExpressionReroute_47}
    TextureInput {expression=&self:MaterialExpressionReroute_48}
   ]}
  MaterialExpressionReroute_47 = Reroute(None=@MaterialExpressionReroute_28.BaseColor)
  n77 = Constant() {r=-0.5}
  MaterialExpressionReroute_48 = Reroute(None=@n11)
  n78 = Power(Base=@MaterialExpressionReroute_48.TextureInput) {constExponent=0.4545}
  n79 = ComponentMask(None=Abs(None=Add(A=@n77, B=@n78))) {g=true}
  n80 = Add(A=@n79, B=@n79)
  n81 = Power(Base=@n79) {constExponent=8}
  n82 = Clamp(None=Power(Base=Add(A=@n80, B=@n81)) {constExponent=0.25}, Min=Constant() {r=0.5}) {clampMode=CMODE_ClampMin, minDefault=0.5}
  n83 = StaticSwitchParameter(True=Multiply(A=@MaterialExpressionReroute_47.BaseColorInput, B=Saturate(None=@n82)), False=@MaterialExpressionReroute_47.BaseColorInput) {
   desc="DO NOT USE FOR NEW ASSETS! This is intended for porting content from Rocket League only."
   group=Legacy_PORTINGONLY
   parameterName=SharpMasks
  }
  PinBase!(Result=@n83) {reroutePins [Result {expression=&self:MaterialExpressionReroute_49}]}
  PinBase!()
  n84 = VectorParameter() {group=Color, parameterName=TrimColor, sortPriority=4, defaultValue=<rgba 0.03 0.03 0.03 0>}
  PinBase!(TeamColor=@n49.RGB, AccentColor=@n50.RGB, PaintColor=@n51.RGB, TrimColor=@n84.RGB) {reroutePins [
    TeamColor {expression=&self:MaterialExpressionReroute_23}
    AccentColor {expression=&self:MaterialExpressionReroute_24}
    PaintColor {expression=&self:MaterialExpressionReroute_25}
    TrimColor {expression=&self:MaterialExpressionReroute_43}
   ]}
  PinBase!() {reroutePins [
    Roughness {expression=&self:MaterialExpressionReroute_2}
    SkinMask {expression=&self:MaterialExpressionReroute_3}
   ]}
  PinBase!(RoughnessMasked=@n70) {reroutePins [RoughnessMasked {expression=&self:MaterialExpressionReroute_5}]}
  PinBase!() {reroutePins [
    TeamColor {expression=&self:MaterialExpressionReroute_8}
    AccentColor {expression=&self:MaterialExpressionReroute_13}
    PaintColor {expression=&self:MaterialExpressionReroute_18}
    SkinMaskTexture {expression=&self:MaterialExpressionReroute_20}
   ]}
  PinBase!(BaseColor=@n8) {reroutePins [BaseColor {expression=&self:MaterialExpressionReroute_26}]}
  MaterialExpressionReroute_43 = Reroute(None=@n84.RGB)
  MaterialExpressionReroute_49 = Reroute(None=@n83)
 ]}
